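# Modular Control Rig biped foot module that adds a reverse-foot IK setup and toe controls.
# Construction resolves joint and optional pivot connectors, reads the parent module's IK metadata, then builds nested pivot nulls, controls and animation channels.
# Pivot controls drive toe-tip, heel and ball nulls, while clamped bank and rocker values roll the foot over its edges, heel and toe.
# The resulting IK null drives the parent's target effector.
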ControlRigBlueprint /EpicControlRig/Modules/Biped/Foot/CRM_Epic_Foot_v04
graphs [
 ConstructionGraph {nodes [
   RigUnit_PrepareForExecution = &RigUnit_PrepareForExecution()
   n1 = /Script/ControlRig.RigUnit_ResolveArrayConnector(Connector=(Type=Connector,Name="Toe Joints"), SkipSocket=False) {is=struct RigVMUnitNode}
   RigVMVariableNode(Value=@n1.Result, ExecuteContext=@RigUnit_PrepareForExecution.ExecutePin) {variable=Toe Joints}
   > RigVMVariableNode(Value=@n1.bIsConnected) {variable=Has Toes}
   > RigVMVariableNode(Value=/Script/ControlRig.RigUnit_ResolveConnector(Connector=(Type=Connector,Name="Foot Joint"), SkipSocket=False) {is=struct RigVMUnitNode}.Result) {variable=Foot Joint}
   > RigVMVariableNode(Value=&RigUnit_ResolveConnector(Connector=(Type=Connector,Name="Ball Joint"), SkipSocket=False).Result) {variable=Ball Joint}
   DISPATCH_RigDispatch_GetModuleMetadata = RigVMDispatchNode(Default=(Type=Bone,Name="None"), Name=IK Null, NameSpace=Parent) {template="DISPATCH_RigDispatch_GetModuleMetadata(in Name,in NameSpace,in Default,out Value,out Found)"}
   > RigVMVariableNode(Value=@DISPATCH_RigDispatch_GetModuleMetadata.Value) {variable=Target Effector}
   > RigVMVariableNode(Value=RigVMDispatchNode(Default=(Type=Bone,Name="None"), Name=IK Control, NameSpace=Parent) {template="DISPATCH_RigDispatch_GetModuleMetadata(in Name,in NameSpace,in Default,out Value,out Found)"}.Value) {variable=Target IK Control}
   > RigVMVariableNode(Value=RigVMDispatchNode(Default=(Type=Bone,Name="None"), Name=IK Driver, NameSpace=Parent) {template="DISPATCH_RigDispatch_GetModuleMetadata(in Name,in NameSpace,in Default,out Value,out Found)"}.Value) {variable=Target IK Driver}
   > RigVMVariableNode(Value=RigVMDispatchNode(Default=true, Name=IK Solve, NameSpace=Parent) {template="DISPATCH_RigDispatch_GetModuleMetadata(in Name,in NameSpace,in Default,out Value,out Found)"}.Value) {variable=IK Solve}
   > RigVMDispatchNode(Array=RigVMVariableNode() {variable=FK Controls}.Value) {template="DISPATCH_RigVMDispatch_ArrayReset(io Array)"}
   > RigVMDispatchNode(Array=RigVMVariableNode() {variable=IK Controls}.Value) {template="DISPATCH_RigVMDispatch_ArrayReset(io Array)"}
   > RigVMDispatchNode(Array=RigVMVariableNode() {variable=Pivot Controls}.Value) {template="DISPATCH_RigVMDispatch_ArrayReset(io Array)"}
   > RigVMVariableNode(Value=&RigUnit_GetTransform(bInitial=true, Item=RigVMVariableNode() {variable=Target IK Control}.Value).Transform.Rotation) {variable=Target IK Orientation}
   > RigVMVariableNode(Value=RigVMVariableNode() {variable=Ball Joint}.Value) {variable=Ball Pivot Bone}
   > RigVMFunction_Sequence = RigVMAggregateNode()
   n2 = RigVMVariableNode() {variable=Foot Rocker Control}
   SpawnAnimationChannel = /Script/ControlRig.RigUnit_HierarchyAddAnimationChannelFloat(
    InitialValue=1
    LimitsEnabled=(Enabled=(bMinimum=True,bMaximum=True))
    MaximumValue=1
    MinimumValue=0
    Name=Rocker Blend
    Parent=@n2.Value
    ExecutePin=@RigVMFunction_Sequence.H
   ) {
    is=struct RigVMUnitNode
    template="SpawnAnimationChannel::Execute(in InitialValue,in MinimumValue,in MaximumValue,in LimitsEnabled,in Parent,in Name,out Item)"
   }
   > RigVMVariableNode(Value=@SpawnAnimationChannel.Item) {variable=Rocker Blend Channel}
   > n3 = &RigUnit_HierarchyAddAnimationChannelFloat(
    InitialValue=45
    LimitsEnabled=(Enabled=(bMinimum=false,bMaximum=false))
    MaximumValue=90
    MinimumValue=0
    Name=Rocker Ball Rotation
    Parent=@n2.Value
   ) {template="SpawnAnimationChannel::Execute(in InitialValue,in MinimumValue,in MaximumValue,in LimitsEnabled,in Parent,in Name,out Item)"}
   > RigVMVariableNode(Value=@n3.Item) {variable=Rocker Ball Rotation Channel}
   n4 = RigVMVariableNode() {variable=Target IK Control}
   > n5 = /Script/ControlRig.RigUnit_HierarchyAddAnimationChannelBool(
    InitialValue=False
    LimitsEnabled=()
    MaximumValue=True
    MinimumValue=False
    Name=Foot Pivot Control Vis
    Parent=RigVMDispatchNode(Condition=/Script/ControlRig.RigUnit_ItemExists(Item=@n4.Value) {is=struct RigVMUnitNode}.Exists, True=@n4.Value, False=@n2.Value) {template="DISPATCH_RigVMDispatch_If(in Condition,in True,in False,out Result)"}.Result
   ) {
    is=struct RigVMUnitNode
    template="SpawnAnimationChannel::Execute(in InitialValue,in MinimumValue,in MaximumValue,in LimitsEnabled,in Parent,in Name,out Item)"
   }
   > RigVMVariableNode(Value=@n5.Item) {variable=Pivot Control Vis Channel}
   n6 = RigVMVariableNode() {variable=Ball IK Shape Settings}
   Get Control Color From Metadata = RigVMFunctionReferenceNode(Color Override=@n6.Value.Color)
   n7 = &RigUnit_GetTransform(bInitial=true, Item=RigVMVariableNode() {variable=Ball Joint}.Value)
   n8 = /Script/ControlRig.RigUnit_HierarchyAddControlTransform(
    InitialValue=(Rotation=(X=0,Y=0,Z=0,W=1),Translation=(X=0,Y=0,Z=0),Scale3D=(X=1,Y=1,Z=1))
    Name=Ball IK
    OffsetSpace=GlobalSpace
    OffsetTransform=(Rotation=(X=0,Y=0,Z=0,W=1),Translation=(X=0,Y=0,Z=0),Scale3D=(X=1,Y=1,Z=1))
    Settings="(InitialSpace=LocalSpace,bUsePreferredRotationOrder=true,PreferredRotationOrder=YZX,Limits=(LimitTranslationX=(bMinimum=False,bMaximum=False),LimitTranslationY=(bMinimum=False,bMaximum=False),LimitTranslationZ=(bMinimum=False,bMaximum=False),LimitPitch=(bM ...(truncated)"
    Settings.Shape.Color=@"Get Control Color From Metadata".Result
    OffsetTransform.Scale3D=@n7.Transform.Scale3D
    OffsetTransform.Translation=@n7.Transform.Translation
    OffsetTransform.Rotation=&RigUnit_GetTransform(bInitial=true, Item=RigVMVariableNode() {variable=IK Null}.Value).Transform.Rotation
    Settings.Shape.Name=@n6.Value.Name
    Settings.Shape.Transform=@n6.Value.Transform
    Settings.Shape.bVisible=@n6.Value.bVisible
    Parent=RigVMVariableNode() {variable=Outer Pivot Null}.Value
    ExecutePin=@RigVMFunction_Sequence.E
   ) {
    is=struct RigVMUnitNode
    template="SpawnControl::Execute(in InitialValue,in Settings,in OffsetTransform,in OffsetSpace,in Parent,in Name,out Item)"
   }
   > RigVMVariableNode(Value=@n8.Item) {variable=Ball IK Control}
   > RigVMDispatchNode(Array=RigVMVariableNode() {variable=IK Controls}.Value, Element=@n8.Item) {template="DISPATCH_RigVMDispatch_ArrayAdd(io Array,in Element,out Index)"}
   n9 = RigVMVariableNode() {variable=Foot Rocker Shape Settings}
   Get Control Color From Metadata_5 = RigVMFunctionReferenceNode(Color Override=@n9.Value.Color)
   > n10 = &RigUnit_HierarchyAddControlTransform(
    InitialValue=(Rotation=(X=0,Y=0,Z=0,W=1),Translation=(X=0,Y=0,Z=0),Scale3D=(X=1,Y=1,Z=1))
    Name=Foot Rocker
    OffsetSpace=GlobalSpace
    OffsetTransform=(Rotation=(X=0,Y=0,Z=0,W=1),Translation=(X=0,Y=0,Z=0),Scale3D=(X=1,Y=1,Z=1))
    Settings="(InitialSpace=LocalSpace,bUsePreferredRotationOrder=true,PreferredRotationOrder=YXZ,Limits=(LimitTranslationX=(bMinimum=False,bMaximum=False),LimitTranslationY=(bMinimum=False,bMaximum=False),LimitTranslationZ=(bMinimum=False,bMaximum=False),LimitPitch=(bM ...(truncated)"
    Settings.Shape.Color=@"Get Control Color From Metadata_5".Result
    OffsetTransform.Rotation=&RigUnit_GetTransform(bInitial=true, Item=RigVMVariableNode() {variable=Ball Pivot Null}.Value).Transform.Rotation
    OffsetTransform.Translation=&RigUnit_GetTransform(bInitial=true, Item=RigVMVariableNode() {variable=Heel Pivot Null}.Value).Transform.Translation
    Settings.Shape.Name=@n9.Value.Name
    Settings.Shape.Transform=@n9.Value.Transform
    Settings.Shape.bVisible=@n9.Value.bVisible
    Parent=RigVMVariableNode() {variable=IK Foot Space}.Value
   ) {template="SpawnControl::Execute(in InitialValue,in Settings,in OffsetTransform,in OffsetSpace,in Parent,in Name,out Item)"}
   > RigVMVariableNode(Value=@n10.Item) {variable=Foot Rocker Control}
   > RigVMDispatchNode(Array=RigVMVariableNode() {variable=IK Controls}.Value, Element=@n10.Item) {template="DISPATCH_RigVMDispatch_ArrayAdd(io Array,in Element,out Index)"}
   n11 = RigVMVariableNode() {variable=Foot Joint}
   > Set_Metadata_2 = RigVMDispatchNode(
    Name=Delta
    NameSpace=Self
    Value=&RigVMFunction_MathTransformMakeRelative(
     Global=&RigUnit_GetTransform(bInitial=true, Item=@n10.Item).Transform
     Parent=&RigUnit_GetTransform(bInitial=true, Item=@n11.Value).Transform
    ).Local
    Item=@n10.Item
   ) {template="DISPATCH_RigDispatch_SetMetadata(in Item,in Name,in NameSpace,in Value,out Success)"}
   n12 = RigVMVariableNode() {variable=Ball Joint}
   n13 = RigVMVariableNode() {variable=Toe Shape Settings}
   Get Control Color From Metadata v02 = RigVMFunctionReferenceNode(Bone=@n12.Value, Color Override=@n13.Value.Color)
   > n14 = &RigUnit_HierarchyAddControlTransform(
    InitialValue=(Rotation=(X=0,Y=0,Z=0,W=1),Translation=(X=0,Y=0,Z=0),Scale3D=(X=1,Y=1,Z=1))
    Name=Toes FK
    OffsetSpace=GlobalSpace
    Settings="(InitialSpace=LocalSpace,bUsePreferredRotationOrder=False,PreferredRotationOrder=YZX,Limits=(LimitTranslationX=(bMinimum=False,bMaximum=False),LimitTranslationY=(bMinimum=False,bMaximum=False),LimitTranslationZ=(bMinimum=False,bMaximum=False),LimitPitch=(b ...(truncated)"
    Settings.Shape.Color=@"Get Control Color From Metadata v02".Result
    OffsetTransform=&RigUnit_GetTransform(bInitial=true, Item=@n12.Value).Transform
    Settings.Shape.Name=@n13.Value.Name
    Settings.Shape.Transform=@n13.Value.Transform
    Settings.Shape.bVisible=@n13.Value.bVisible
    Parent=RigVMVariableNode() {variable=Foot Joint}.Value
   ) {template="SpawnControl::Execute(in InitialValue,in Settings,in OffsetTransform,in OffsetSpace,in Parent,in Name,out Item)"}
   > RigVMVariableNode(Value=@n14.Item) {variable=Toes FK Control}
   > Set Mirror Behavior_1 = RigVMFunctionReferenceNode(Value=RigVMVariableNode() {variable=FK Control Mirror Behavior}.Value, Item=@n14.Item)
   > n15 = &RigUnit_HierarchyAddNull(
    Name=Toes IK Null
    Space=LocalSpace
    Transform=(Rotation=(X=0,Y=0,Z=0,W=1),Translation=(X=0,Y=0,Z=0),Scale3D=(X=1,Y=1,Z=1))
    Parent=RigVMVariableNode() {variable=Foot Joint}.Value
   )
   > RigVMVariableNode(Value=@n15.Item) {variable=Toes IK Null}
   n16 = RigVMVariableNode() {variable=Toe Shape Settings}
   n17 = RigVMVariableNode() {variable=Ball Joint}
   Get Control Color From Metadata v02_1 = RigVMFunctionReferenceNode(Color Override=@n16.Value.Color, Bone=@n17.Value)
   > n18 = &RigUnit_HierarchyAddControlTransform(
    InitialValue=(Rotation=(X=0,Y=0,Z=0,W=1),Translation=(X=0,Y=0,Z=0),Scale3D=(X=1,Y=1,Z=1))
    Name=Toes IK
    OffsetSpace=GlobalSpace
    Settings="(InitialSpace=LocalSpace,bUsePreferredRotationOrder=False,PreferredRotationOrder=YZX,Limits=(LimitTranslationX=(bMinimum=False,bMaximum=False),LimitTranslationY=(bMinimum=False,bMaximum=False),LimitTranslationZ=(bMinimum=False,bMaximum=False),LimitPitch=(b ...(truncated)"
    Settings.Shape.Color=@"Get Control Color From Metadata v02_1".Result
    OffsetTransform=&RigUnit_GetTransform(bInitial=true, Item=@n17.Value).Transform
    Parent=@n15.Item
    Settings.Shape.Name=@n16.Value.Name
    Settings.Shape.Transform=@n16.Value.Transform
    Settings.Shape.bVisible=@n16.Value.bVisible
   ) {template="SpawnControl::Execute(in InitialValue,in Settings,in OffsetTransform,in OffsetSpace,in Parent,in Name,out Item)"}
   > RigVMVariableNode(Value=@n18.Item) {variable=Toes IK Control}
   > Set Mirror Behavior_2 = RigVMFunctionReferenceNode(Value=RigVMVariableNode() {variable=FK Control Mirror Behavior}.Value, Item=@n18.Item)
   n19 = RigVMVariableNode() {variable=Pivot Shape Settings}
   Get Control Color From Metadata_1 = RigVMFunctionReferenceNode(Color Override=@n19.Value.Color)
   SpawnControl = &RigUnit_HierarchyAddControlTransform(
    InitialValue=(Rotation=(X=0,Y=0,Z=0,W=1),Translation=(X=0,Y=0,Z=0),Scale3D=(X=1,Y=1,Z=1))
    Name=Toe Tip
    OffsetSpace=GlobalSpace
    Settings="(InitialSpace=LocalSpace,bUsePreferredRotationOrder=true,PreferredRotationOrder=ZYX,Limits=(LimitTranslationX=(bMinimum=False,bMaximum=False),LimitTranslationY=(bMinimum=False,bMaximum=False),LimitTranslationZ=(bMinimum=False,bMaximum=False),LimitPitch=(bM ...(truncated)"
    Settings.Shape.Color=@"Get Control Color From Metadata_1".Result
    OffsetTransform=&RigUnit_GetTransform(bInitial=true, Item=RigVMVariableNode() {variable=Toe Tip Pivot Null}.Value).Transform
    Settings.Shape.Name=@n19.Value.Name
    Settings.Shape.Transform=@n19.Value.Transform
    Settings.Shape.bVisible=@n19.Value.bVisible
    Parent=RigVMVariableNode() {variable=IK Foot Space}.Value
    ExecutePin=@RigVMFunction_Sequence.B
   ) {template="SpawnControl::Execute(in InitialValue,in Settings,in OffsetTransform,in OffsetSpace,in Parent,in Name,out Item)"}
   > RigVMVariableNode(Value=@SpawnControl.Item) {variable=Toe Tip Control}
   > RigVMDispatchNode(Array=RigVMVariableNode() {variable=IK Controls}.Value, Element=@SpawnControl.Item) {template="DISPATCH_RigVMDispatch_ArrayAdd(io Array,in Element,out Index)"}
   > RigVMDispatchNode(Array=RigVMVariableNode() {variable=Pivot Controls}.Value, Element=@SpawnControl.Item) {template="DISPATCH_RigVMDispatch_ArrayAdd(io Array,in Element,out Index)"}
   n20 = RigVMVariableNode() {variable=Ball Joint}
   > Set_Metadata = RigVMDispatchNode(
    Name=Delta
    NameSpace=Self
    Value=&RigVMFunction_MathTransformMakeRelative(
     Global=&RigUnit_GetTransform(bInitial=true, Item=@SpawnControl.Item).Transform
     Parent=&RigUnit_GetTransform(bInitial=true, Item=@n20.Value).Transform
    ).Local
    Item=@SpawnControl.Item
   ) {template="DISPATCH_RigDispatch_SetMetadata(in Item,in Name,in NameSpace,in Value,out Success)"}
   n21 = &RigUnit_ResolveConnector(Connector=(Type=Connector,Name="Heel Pivot"), SkipSocket=False)
   n22 = RigVMVariableNode() {variable=Footprint Space Transform}
   n23 = RigVMVariableNode() {variable=Heel Pivot Offset}
   > n24 = &RigUnit_HierarchyAddNull(
    Name=Heel Pivot
    Space=GlobalSpace
    Transform=(Rotation=(X=0,Y=0,Z=0,W=1),Translation=(X=0,Y=0,Z=0),Scale3D=(X=1,Y=1,Z=1))
    Transform.Translation=RigVMDispatchNode(
     True=&RigUnit_GetTransform(bInitial=true, Item=@n21.Result).Transform
     False=/Script/RigVM.RigVMFunction_MathTransformMakeAbsolute(Local=(Rotation=(X=0,Y=0,Z=0,W=1),Translation=(X=0,Y=0,Z=0),Scale3D=(X=1,Y=1,Z=1)), Parent=@n22.Value, Local.Translation=@n23.Value) {is=struct RigVMUnitNode, template="Make Absolute::Execute(in Local,in Parent,out Global)"}.Global
     Condition=@n21.bIsConnected
    ) {template="DISPATCH_RigVMDispatch_If(in Condition,in True,in False,out Result)"}.Result.Translation
    Transform.Rotation=RigVMVariableNode() {variable=Target IK Orientation}.Value
    Parent=RigVMVariableNode() {variable=Toe Tip Pivot Null}.Value
   )
   > RigVMVariableNode(Value=@n24.Item) {variable=Running Parent}
   > RigVMVariableNode(Value=@n24.Item) {variable=Heel Pivot Null}
   > n25 = &RigUnit_HierarchyAddNull(Name=Ball Pivot, Space=GlobalSpace, Transform=RigVMVariableNode() {variable=Footprint Space Transform}.Value, Parent=@n24.Item)
   > RigVMVariableNode(Value=@n25.Item) {variable=Ball Pivot Null}
   > n26 = &RigUnit_HierarchyAddNull(
    Name=Toe Tip Rocker Pivot
    Space=GlobalSpace
    Transform=&RigUnit_GetTransform(bInitial=true, Item=RigVMVariableNode() {variable=Toe Tip Pivot Null}.Value).Transform
    Parent=@n25.Item
   )
   > RigVMVariableNode(Value=@n26.Item) {variable=Toe Tip Rocker Pivot}
   > n27 = &RigUnit_HierarchyAddNull(
    Name=Heel Rocker Pivot
    Space=GlobalSpace
    Transform=&RigUnit_GetTransform(bInitial=true, Item=RigVMVariableNode() {variable=Heel Pivot Null}.Value).Transform
    Parent=@n26.Item
   )
   > RigVMVariableNode(Value=@n27.Item) {variable=Heel Rocker Pivot}
   n28 = &RigUnit_GetTransform(bInitial=true, Item=RigVMVariableNode() {variable=Ball Pivot Bone}.Value)
   n29 = &RigUnit_HierarchyAddNull(
    Name=Heel Lift
    Space=GlobalSpace
    Transform=(Rotation=(X=0,Y=0,Z=0,W=1),Translation=(X=0,Y=0,Z=0),Scale3D=(X=1,Y=1,Z=1))
    Transform.Scale3D=@n28.Transform.Scale3D
    Transform.Translation=@n28.Transform.Translation
    Parent=RigVMVariableNode() {variable=Outer Pivot Null}.Value
    ExecutePin=@RigVMFunction_Sequence.D
    Transform.Rotation=RigVMVariableNode() {variable=Target IK Orientation}.Value
   )
   > RigVMVariableNode(Value=@n29.Item) {variable=Heel Lift Null}
   RigUnit_GetTransform = &RigUnit_GetTransform(bInitial=true, Item=RigVMVariableNode() {variable=Ball Pivot Bone}.Value)
   > n30 = &RigUnit_HierarchyAddNull(
    Name=Ball IK
    Space=GlobalSpace
    Transform=(Rotation=(X=0,Y=0,Z=0,W=1),Translation=(X=0,Y=0,Z=0),Scale3D=(X=1,Y=1,Z=1))
    Transform.Scale3D=@RigUnit_GetTransform.Transform.Scale3D
    Transform.Translation=@RigUnit_GetTransform.Transform.Translation
    Parent=RigVMVariableNode() {variable=Outer Pivot Null}.Value
    Transform.Rotation=RigVMVariableNode() {variable=Target IK Orientation}.Value
   )
   > RigVMVariableNode(Value=@n30.Item) {variable=Ball IK Null}
   > n31 = &RigUnit_HierarchyAddNull(
    Name=Toe IK
    Space=GlobalSpace
    Transform=&RigUnit_GetTransform(bInitial=true, Item=RigVMVariableNode() {variable=Toe Tip Pivot Null}.Value).Transform
    Parent=RigVMVariableNode() {variable=Outer Pivot Null}.Value
   )
   > RigVMVariableNode(Value=@n31.Item) {variable=Toe IK Null}
   > n32 = &RigUnit_HierarchyAddNull(
    Name=IK
    Space=GlobalSpace
    Transform=&RigUnit_GetTransform(bInitial=true, Item=RigVMVariableNode() {variable=Target Effector}.Value).Transform
    Parent=RigVMVariableNode() {variable=Heel Lift Null}.Value
   )
   > RigVMVariableNode(Value=@n32.Item) {variable=IK Null}
   n33 = &RigVMFunction_ControlFlowBranch(Condition=RigVMVariableNode() {variable=Has Toes}.Value, ExecuteContext=@RigVMFunction_Sequence.F)
   n34 = RigVMDispatchNode(ExecuteContext=@n33.True, Array=RigVMVariableNode() {variable=Toe Joints}.Value) {template="DISPATCH_RigVMDispatch_ArrayIterator(in Array,out Element,out Index,out Count,out Ratio)"}
   Int to Name_2 = /Script/RigVM.RigVMFunction_MathIntToName(PaddedSize=0, Number=@n34.Index) {is=struct RigVMUnitNode, template="Int to Name::Execute(in Number,in PaddedSize,out Result)"}
   n35 = RigVMVariableNode() {variable=Toe Shape Settings}
   Get Control Color From Metadata_3 = RigVMFunctionReferenceNode(Color Override=@n35.Value.Color)
   > n36 = &RigUnit_HierarchyAddControlTransform(
    InitialValue=(Rotation=(X=0,Y=0,Z=0,W=1),Translation=(X=0,Y=0,Z=0),Scale3D=(X=1,Y=1,Z=1))
    OffsetSpace=GlobalSpace
    Settings="(InitialSpace=LocalSpace,bUsePreferredRotationOrder=False,PreferredRotationOrder=YZX,Limits=(LimitTranslationX=(bMinimum=False,bMaximum=False),LimitTranslationY=(bMinimum=False,bMaximum=False),LimitTranslationZ=(bMinimum=False,bMaximum=False),LimitPitch=(b ...(truncated)"
    Name=/Script/RigVM.RigVMFunction_NameConcat(A="FK ", B=@"Int to Name_2".Result) {is=struct RigVMUnitNode, template="Concat::Execute(in A,in B,out Result)"}.Result
    Settings.Shape.Color=@"Get Control Color From Metadata_3".Result
    OffsetTransform=&RigUnit_GetTransform(bInitial=true, Item=@n34.Element).Transform
    Parent=RigVMVariableNode() {variable=Toes FK Control}.Value
    Settings.Shape.Name=@n35.Value.Name
    Settings.Shape.Transform=@n35.Value.Transform
    Settings.Shape.bVisible=@n35.Value.bVisible
   ) {template="SpawnControl::Execute(in InitialValue,in Settings,in OffsetTransform,in OffsetSpace,in Parent,in Name,out Item)"}
   > RigVMDispatchNode(Element=@n36.Item, Array=RigVMVariableNode() {variable=FK Controls}.Value) {template="DISPATCH_RigVMDispatch_ArrayAdd(io Array,in Element,out Index)"}
   > Set Mirror Behavior = RigVMFunctionReferenceNode(Value=RigVMVariableNode() {variable=FK Control Mirror Behavior}.Value, Item=@n36.Item)
   n37 = &RigUnit_HierarchyAddNull(
    Name=IK Foot Space
    Space=LocalSpace
    Transform=(Rotation=(X=0,Y=0,Z=0,W=1),Translation=(X=0,Y=0,Z=0),Scale3D=(X=1,Y=1,Z=1))
    Parent=RigVMVariableNode() {variable=Target IK Driver}.Value
    ExecutePin=@RigVMFunction_Sequence.A
   )
   > RigVMVariableNode(Value=@n37.Item) {variable=IK Foot Space}
   n38 = &RigUnit_GetTransform(bInitial=true, Item=RigVMVariableNode() {variable=Ball Pivot Bone}.Value)
   > RigVMVariableNode(
    Value=(Rotation=(X=0,Y=0,Z=0,W=1),Translation=(X=0,Y=0,Z=0),Scale3D=(X=1,Y=1,Z=1))
    Value.Scale3D=@n38.Transform.Scale3D
    Value.Translation.X=@n38.Transform.Translation.X
    Value.Translation.Y=@n38.Transform.Translation.Y
    Value.Rotation=RigVMVariableNode() {variable=Target IK Orientation}.Value
   ) {variable=Footprint Space Transform}
   n39 = RigVMVariableNode() {variable=Footprint Space Transform}
   n40 = RigVMVariableNode() {variable=Toe Tip Pivot Offset}
   n41 = &RigUnit_ResolveConnector(Connector=(Type=Connector,Name="Toe Tip Pivot"), SkipSocket=False)
   > n42 = &RigUnit_HierarchyAddNull(
    Name=Toe Tip Pivot
    Space=GlobalSpace
    Transform=(Rotation=(X=0,Y=0,Z=0,W=1),Translation=(X=0,Y=0,Z=0),Scale3D=(X=1,Y=1,Z=1))
    Transform.Translation=RigVMDispatchNode(
     False=&RigVMFunction_MathTransformMakeAbsolute(Local=(Rotation=(X=0,Y=0,Z=0,W=1),Translation=(X=0,Y=0,Z=0),Scale3D=(X=1,Y=1,Z=1)), Parent=@n39.Value, Local.Translation=@n40.Value) {template="Make Absolute::Execute(in Local,in Parent,out Global)"}.Global
     Condition=@n41.bIsConnected
     True=&RigUnit_GetTransform(bInitial=true, Item=@n41.Result).Transform
    ) {template="DISPATCH_RigVMDispatch_If(in Condition,in True,in False,out Result)"}.Result.Translation
    Transform.Rotation=RigVMVariableNode() {variable=Target IK Orientation}.Value
    Parent=RigVMVariableNode() {variable=IK Foot Space}.Value
   )
   > RigVMVariableNode(Value=@n42.Item) {variable=Toe Tip Pivot Null}
   RigVMFunction_MathVectorDistance = &RigVMFunction_MathVectorDistance(
    A=&RigUnit_GetTransform(bInitial=true, Item=RigVMVariableNode() {variable=Inner Pivot Null}.Value).Transform.Translation
    B=&RigUnit_GetTransform(bInitial=true, Item=RigVMVariableNode() {variable=Outer Pivot Null}.Value).Transform.Translation
   )
   Divide = &RigVMFunction_MathDoubleDiv(B=8, A=@RigVMFunction_MathVectorDistance.Result)
   n43 = &RigUnit_GetTransform(bInitial=true, Item=RigVMVariableNode() {variable=Toe Tip Pivot Null}.Value)
   n44 = &RigUnit_GetTransform(bInitial=true, Item=RigVMVariableNode() {variable=Heel Pivot Null}.Value)
   Interpolate = /Script/RigVM.RigVMFunction_MathTransformLerp(T=0.5, A=@n43.Transform, B=@n44.Transform) {is=struct RigVMUnitNode, template="Interpolate::Execute(in A,in B,in T,out Result)"}
   n45 = &RigUnit_HierarchyAddControlTransform(
    InitialValue=(Rotation=(X=0,Y=0,Z=0,W=1),Translation=(X=0,Y=0,Z=0),Scale3D=(X=1,Y=1,Z=1))
    Name=Footprint Display
    OffsetSpace=GlobalSpace
    Settings="(InitialSpace=LocalSpace,bUsePreferredRotationOrder=False,PreferredRotationOrder=YZX,Limits=(LimitTranslationX=(bMinimum=False,bMaximum=False),LimitTranslationY=(bMinimum=False,bMaximum=False),LimitTranslationZ=(bMinimum=False,bMaximum=False),LimitPitch=(b ...(truncated)"
    Settings.Shape.Transform.Scale3D.X=@Divide.Result
    Settings.Shape.Transform.Scale3D.Y=&RigVMFunction_MathDoubleDiv(B=8, A=&RigVMFunction_MathVectorDistance(A=@n43.Transform.Translation, B=@n44.Transform.Translation).Result).Result
    OffsetTransform=@Interpolate.Result
    Parent=RigVMVariableNode() {variable=Target IK Control}.Value
    ExecutePin=@RigVMFunction_Sequence.G
   ) {template="SpawnControl::Execute(in InitialValue,in Settings,in OffsetTransform,in OffsetSpace,in Parent,in Name,out Item)"}
   > n46 = RigVMVariableNode(Value=@n45.Item) {variable=Footprint Control}
   n47 = RigVMVariableNode() {variable=Target IK Control}
   RigVMFunction_ControlFlowBranch = &RigVMFunction_ControlFlowBranch(Condition=&RigUnit_ItemExists(Item=@n47.Value).Exists, ExecuteContext=@n46.ExecuteContext)
   n48 = &RigUnit_HierarchyAddAnimationChannelBool(
    InitialValue=false
    LimitsEnabled=()
    MaximumValue=True
    MinimumValue=False
    Name=Footprint Vis
    ExecutePin=@RigVMFunction_ControlFlowBranch.True
    Parent=@n47.Value
   ) {template="SpawnAnimationChannel::Execute(in InitialValue,in MinimumValue,in MaximumValue,in LimitsEnabled,in Parent,in Name,out Item)"}
   > RigVMVariableNode(Value=@n48.Item) {variable=Footprint Vis Animation Channel}
   RigUnit_ItemToName = /Script/ControlRig.RigUnit_ItemToName(Value=RigVMVariableNode() {variable=Footprint Control}.Value) {is=struct RigVMUnitNode, template="Cast::Execute(in Value,out Result)"}
   > SetControlColor = /Script/ControlRig.RigUnit_SetControlColor(
    Color=/Script/ControlRig.RigUnit_HierarchyGetShapeSettings(Item=@n47.Value) {is=struct RigVMUnitNode}.Settings.Color
    Control=@RigUnit_ItemToName.Result
   ) {is=struct RigVMUnitNode, template="SetControlColor::Execute(in Control,in Color)"}
   Clone = RigVMDispatchNode(Array=RigVMVariableNode() {variable=FK Controls}.Value) {template="DISPATCH_RigVMDispatch_ArrayClone(in Array,out Clone)"}
   Make_Array_1 = RigVMDispatchNode(
    Values=((Type=None,Name="None"),((),Type=None,Name="None"),((),Type=None,Name="None"))
    Values.0=RigVMVariableNode() {variable=Toes FK Control}.Value
    Values.2=RigVMVariableNode() {variable=Foot Rocker Control}.Value
    Values.1=RigVMVariableNode() {variable=Ball IK Control}.Value
   ) {template="DISPATCH_RigVMDispatch_ArrayMake(in Values,out Array)"}
   n49 = RigVMDispatchNode(Array=@Clone.Clone, Other=@Make_Array_1.Array, ExecuteContext=@RigVMFunction_Sequence.I) {template="DISPATCH_RigVMDispatch_ArrayUnion(io Array,in Other)"}
   > n50 = RigVMDispatchNode(Array=@n49.Array, Other=RigVMVariableNode() {variable=Pivot Controls}.Value) {template="DISPATCH_RigVMDispatch_ArrayUnion(io Array,in Other)"}
   Make_Array = RigVMDispatchNode(
    Values=((Type=None,Name="None"),((),Type=None,Name="None"))
    Values.0=RigVMVariableNode() {variable=Foot Joint}.Value
    Values.1=RigVMVariableNode() {variable=Ball Joint}.Value
   ) {template="DISPATCH_RigVMDispatch_ArrayMake(in Values,out Array)"}
   > Get Chain Length = RigVMFunctionReferenceNode(Initial=true, Items=@Make_Array.Array)
   Get_Module_Metadata_1 = RigVMDispatchNode(Default=1, Name=Global Control Scale, NameSpace=Root) {template="DISPATCH_RigDispatch_GetModuleMetadata(in Name,in NameSpace,in Default,out Value,out Found)"}
   Num = RigVMDispatchNode(Array=@Make_Array.Array) {template="DISPATCH_RigVMDispatch_ArrayGetNum(in Array,out Num)"}
   > Set Control Scale = RigVMFunctionReferenceNode(
    Global Scale=@Get_Module_Metadata_1.Value
    Auto Scale Factor=RigVMDispatchNode(
     False=1
     True=&RigVMFunction_MathDoubleDiv(B=17, A=@"Get Chain Length".Length).Result
     Condition=/Script/RigVM.RigVMFunction_MathIntGreater(B=1, A=@Num.Num) {is=struct RigVMUnitNode, template="Greater::Execute(in A,in B,out Result)"}.Result
    ) {template="DISPATCH_RigVMDispatch_If(in Condition,in True,in False,out Result)"}.Result
    Scale=RigVMVariableNode() {variable=Control Scale}.Value
    Controls=@n50.Array
   )
   n51 = RigVMVariableNode() {variable=Footprint Space Transform}
   n52 = RigVMVariableNode() {variable=Inner Bank Offset}
   n53 = &RigUnit_ResolveConnector(Connector=(Type=Connector,Name="Inner Bank Pivot"), SkipSocket=False)
   n54 = &RigUnit_HierarchyAddNull(
    Name=Inner Pivot
    Space=GlobalSpace
    Transform=(Rotation=(X=0,Y=0,Z=0,W=1),Translation=(X=0,Y=0,Z=0),Scale3D=(X=1,Y=1,Z=1))
    Transform.Translation=RigVMDispatchNode(
     False=&RigVMFunction_MathTransformMakeAbsolute(Local=(Rotation=(X=0,Y=0,Z=0,W=1),Translation=(X=0,Y=0,Z=0),Scale3D=(X=1,Y=1,Z=1)), Parent=@n51.Value, Local.Translation=@n52.Value) {template="Make Absolute::Execute(in Local,in Parent,out Global)"}.Global
     Condition=@n53.bIsConnected
     True=&RigUnit_GetTransform(bInitial=true, Item=@n53.Result).Transform
    ) {template="DISPATCH_RigVMDispatch_If(in Condition,in True,in False,out Result)"}.Result.Translation
    Transform.Rotation=RigVMVariableNode() {variable=Target IK Orientation}.Value
    Parent=RigVMVariableNode() {variable=Heel Rocker Pivot}.Value
    ExecutePin=@RigVMFunction_Sequence.C
   )
   > RigVMVariableNode(Value=@n54.Item) {variable=Inner Pivot Null}
   n55 = RigVMVariableNode() {variable=Outer Bank Offset}
   n56 = &RigUnit_ResolveConnector(Connector=(Type=Connector,Name="Outer Bank Pivot"), SkipSocket=False)
   > n57 = &RigUnit_HierarchyAddNull(
    Name=Outer Pivot
    Space=GlobalSpace
    Transform=(Rotation=(X=0,Y=0,Z=0,W=1),Translation=(X=0,Y=0,Z=0),Scale3D=(X=1,Y=1,Z=1))
    Parent=@n54.Item
    Transform.Translation=RigVMDispatchNode(
     False=&RigVMFunction_MathTransformMakeAbsolute(Local=(Rotation=(X=0,Y=0,Z=0,W=1),Translation=(X=0,Y=0,Z=0),Scale3D=(X=1,Y=1,Z=1)), Parent=@n51.Value, Local.Translation=@n55.Value) {template="Make Absolute::Execute(in Local,in Parent,out Global)"}.Global
     True=&RigUnit_GetTransform(bInitial=true, Item=@n56.Result).Transform
     Condition=@n56.bIsConnected
    ) {template="DISPATCH_RigVMDispatch_If(in Condition,in True,in False,out Result)"}.Result.Translation
    Transform.Rotation=RigVMVariableNode() {variable=Target IK Orientation}.Value
   )
   > RigVMVariableNode(Value=@n57.Item) {variable=Outer Pivot Null}
   n58 = RigVMVariableNode() {variable=Pivot Shape Settings}
   Get Control Color From Metadata_2 = RigVMFunctionReferenceNode(Color Override=@n58.Value.Color)
   > n59 = &RigUnit_HierarchyAddControlTransform(
    InitialValue=(Rotation=(X=0,Y=0,Z=0,W=1),Translation=(X=0,Y=0,Z=0),Scale3D=(X=1,Y=1,Z=1))
    Name=Heel
    OffsetSpace=GlobalSpace
    Settings="(InitialSpace=LocalSpace,bUsePreferredRotationOrder=true,PreferredRotationOrder=ZYX,Limits=(LimitTranslationX=(bMinimum=False,bMaximum=False),LimitTranslationY=(bMinimum=False,bMaximum=False),LimitTranslationZ=(bMinimum=False,bMaximum=False),LimitPitch=(bM ...(truncated)"
    Settings.Shape.Color=@"Get Control Color From Metadata_2".Result
    OffsetTransform=&RigUnit_GetTransform(bInitial=true, Item=RigVMVariableNode() {variable=Heel Pivot Null}.Value).Transform
    Parent=RigVMVariableNode() {variable=Toe Tip Pivot Null}.Value
    Settings.Shape.Name=@n58.Value.Name
    Settings.Shape.Transform=@n58.Value.Transform
    Settings.Shape.bVisible=@n58.Value.bVisible
   ) {template="SpawnControl::Execute(in InitialValue,in Settings,in OffsetTransform,in OffsetSpace,in Parent,in Name,out Item)"}
   > RigVMVariableNode(Value=@n59.Item) {variable=Heel Control}
   > Add = RigVMDispatchNode(Array=RigVMVariableNode() {variable=IK Controls}.Value, Element=@n59.Item) {template="DISPATCH_RigVMDispatch_ArrayAdd(io Array,in Element,out Index)"}
   > RigVMDispatchNode(Array=RigVMVariableNode() {variable=Pivot Controls}.Value, Element=@n59.Item) {template="DISPATCH_RigVMDispatch_ArrayAdd(io Array,in Element,out Index)"}
   n60 = RigVMVariableNode() {variable=Foot Joint}
   > Set_Metadata_1 = RigVMDispatchNode(
    Name=Delta
    NameSpace=Self
    Value=&RigVMFunction_MathTransformMakeRelative(
     Global=&RigUnit_GetTransform(bInitial=true, Item=@n59.Item).Transform
     Parent=&RigUnit_GetTransform(bInitial=true, Item=@n60.Value).Transform
    ).Local
    Item=@n59.Item
   ) {template="DISPATCH_RigDispatch_SetMetadata(in Item,in Name,in NameSpace,in Value,out Success)"}
   n61 = RigVMVariableNode() {variable=Pivot Shape Settings}
   Get Control Color From Metadata_6 = RigVMFunctionReferenceNode(Color Override=@n61.Value.Color)
   n62 = &RigUnit_GetTransform(bInitial=true, Item=RigVMVariableNode() {variable=Ball Pivot Null}.Value)
   n63 = &RigVMFunction_MathDoubleMul(B=1.2, A=RigVMVariableNode() {variable=Ball IK Shape Settings}.Value.Transform.Scale3D.X)
   > n64 = &RigUnit_HierarchyAddControlTransform(
    InitialValue=(Rotation=(X=0,Y=0,Z=0,W=1),Translation=(X=0,Y=0,Z=0),Scale3D=(X=1,Y=1,Z=1))
    Name=Ball
    OffsetSpace=GlobalSpace
    OffsetTransform=(Rotation=(X=0,Y=0,Z=0,W=1),Translation=(X=0,Y=0,Z=0),Scale3D=(X=1,Y=1,Z=1))
    Settings="(InitialSpace=LocalSpace,bUsePreferredRotationOrder=true,PreferredRotationOrder=ZYX,Limits=(LimitTranslationX=(bMinimum=False,bMaximum=False),LimitTranslationY=(bMinimum=False,bMaximum=False),LimitTranslationZ=(bMinimum=False,bMaximum=False),LimitPitch=(bM ...(truncated)"
    Settings.Shape.Color=@"Get Control Color From Metadata_6".Result
    OffsetTransform.Rotation=@n62.Transform.Rotation
    OffsetTransform.Scale3D=@n62.Transform.Scale3D
    OffsetTransform.Translation.X=@n62.Transform.Translation.X
    OffsetTransform.Translation.Y=@n62.Transform.Translation.Y
    Settings.Shape.Transform.Scale3D.X=@n63.Result
    Settings.Shape.Transform.Scale3D.Y=@n63.Result
    Settings.Shape.Transform.Scale3D.Z=@n63.Result
    Settings.Shape.bVisible=@n61.Value.bVisible
    Parent=RigVMVariableNode() {variable=Heel Pivot Null}.Value
   ) {template="SpawnControl::Execute(in InitialValue,in Settings,in OffsetTransform,in OffsetSpace,in Parent,in Name,out Item)"}
   > RigVMVariableNode(Value=@n64.Item) {variable=Ball Pivot Control}
   > RigVMDispatchNode(Element=@n64.Item, Array=RigVMVariableNode() {variable=IK Controls}.Value) {template="DISPATCH_RigVMDispatch_ArrayAdd(io Array,in Element,out Index)"}
   > RigVMDispatchNode(Element=@n64.Item, Array=RigVMVariableNode() {variable=Pivot Controls}.Value) {template="DISPATCH_RigVMDispatch_ArrayAdd(io Array,in Element,out Index)"}
   RigVMVariableNode() {variable=Ball IK Control}
   RigVMVariableNode() {variable=Ball Pivot Null}
   RigVMVariableNode() {variable=Target IK Orientation}
  ]}
 New Function_ContainedGraph {nodes [
   Entry = RigVMFunctionEntryNode()
   RigUnit_GetTransform = &RigUnit_GetTransform(Space=LocalSpace, bInitial=true, Item=@Entry.Item)
   > RigUnit_SetTransform = &RigUnit_SetTransform(
    Space=LocalSpace
    Weight=1
    bInitial=False
    Item=@Entry.Item
    Value=&RigVMFunction_MathTransformMakeAbsolute(Local=@Entry.Local, Parent=@RigUnit_GetTransform.Transform) {template="Make Absolute::Execute(in Local,in Parent,out Global)"}.Global
   )
   > RigVMFunctionReturnNode()
  ]}
 New Function_ContainedGraph {nodes [
   Entry = RigVMFunctionEntryNode()
   > RigVMFunction_Sequence = RigVMAggregateNode()
   RigVMFunctionReturnNode(ExecuteContext=@RigVMFunction_Sequence.E)
   n1 = RigVMVariableNode() {variable=Foot Bank}
   Clamp = &RigVMFunction_MathDoubleClamp(Maximum=0, Minimum=-90, Value=@n1.Value)
   RigVMFunction_MathQuaternionFromEuler = &RigVMFunction_MathQuaternionFromEuler(
    Euler=(X=0,Y=0,Z=0)
    RotationOrder=XYZ
    Euler.Y=RigVMDispatchNode(
     False=@Clamp.Result
     True=&RigVMFunction_MathDoubleClamp(Maximum=90, Value=@n1.Value).Result
     Condition=RigVMVariableNode() {variable=Negative Side}.Value
    ) {template="DISPATCH_RigVMDispatch_If(in Condition,in True,in False,out Result)"}.Result
   )
   Set Transform_1 = &RigUnit_SetRotation(
    Weight=1
    bInitial=False
    bPropagateToChildren=true
    Value=@RigVMFunction_MathQuaternionFromEuler.Result
    Item=RigVMVariableNode() {variable=Inner Pivot Null}.Value
    ExecutePin=@RigVMFunction_Sequence.B
   )
   n2 = RigVMVariableNode() {variable=Foot Bank}
   n3 = &RigVMFunction_MathDoubleClamp(Maximum=90, Value=@n2.Value)
   n4 = &RigVMFunction_MathDoubleClamp(Maximum=0, Minimum=-90, Value=@n2.Value)
   n5 = RigVMVariableNode() {variable=Negative Side}
   > Set Transform_1_1 = &RigUnit_SetRotation(
    Weight=1
    bInitial=False
    Value=&RigVMFunction_MathQuaternionFromEuler(
     Euler=(X=0,Y=0,Z=0)
     RotationOrder=XYZ
     Euler.Y=RigVMDispatchNode(False=@n3.Result, True=@n4.Result, Condition=@n5.Value) {template="DISPATCH_RigVMDispatch_If(in Condition,in True,in False,out Result)"}.Result
    ).Result
    Item=RigVMVariableNode() {variable=Outer Pivot Null}.Value
   )
   n6 = &RigUnit_GetTransform(Space=LocalSpace, bInitial=False, Item=RigVMVariableNode() {variable=Toe Tip Control}.Value)
   Rotate Aroound Free Pivot = RigVMFunctionReferenceNode(
    Pivot Transform=@n6.Transform
    Rotation Matrix=@n6.Transform
    Input Transform=&RigUnit_GetTransform(Space=LocalSpace, bInitial=true, Item=RigVMVariableNode() {variable=Toe Tip Pivot Null}.Value).Transform
   )
   &RigUnit_SetTransform(
    Space=LocalSpace
    Weight=1
    bInitial=False
    Value=&RigVMFunction_MathTransformMul(A=(Rotation=(X=0,Y=0,Z=0,W=1),Translation=(X=0,Y=0,Z=0),Scale3D=(X=1,Y=1,Z=1)), B=@"Rotate Aroound Free Pivot".Output Transform) {template="Multiply::Execute(in A,in B,out Result)"}.Result
    Item=RigVMVariableNode() {variable=Toe Tip Pivot Null}.Value
    ExecutePin=@RigVMFunction_Sequence.A
   )
   n7 = &RigUnit_GetTransform(Space=LocalSpace, bInitial=False, Item=RigVMVariableNode() {variable=Heel Control}.Value)
   Rotate Aroound Free Pivot_1 = RigVMFunctionReferenceNode(
    Pivot Transform=@n7.Transform
    Rotation Matrix=@n7.Transform
    Input Transform=&RigUnit_GetTransform(Space=LocalSpace, bInitial=true, Item=RigVMVariableNode() {variable=Heel Pivot Null}.Value).Transform
   )
   > &RigUnit_SetTransform(
    Space=LocalSpace
    Weight=1
    bInitial=False
    Value=&RigVMFunction_MathTransformMul(A=(Rotation=(X=0,Y=0,Z=0,W=1),Translation=(X=0,Y=0,Z=0),Scale3D=(X=1,Y=1,Z=1)), B=@"Rotate Aroound Free Pivot_1".Output Transform) {template="Multiply::Execute(in A,in B,out Result)"}.Result
    Item=RigVMVariableNode() {variable=Heel Pivot Null}.Value
   )
   n8 = &RigUnit_GetTransform(Space=LocalSpace, bInitial=False, Item=RigVMVariableNode() {variable=Ball Pivot Control}.Value)
   n9 = RigVMVariableNode() {variable=Ball Pivot Null}
   Rotate Aroound Free Pivot_2 = RigVMFunctionReferenceNode(
    Pivot Transform=@n8.Transform
    Rotation Matrix=@n8.Transform
    Input Transform=&RigUnit_GetTransform(Space=LocalSpace, bInitial=true, Item=@n9.Value).Transform
   )
   > &RigUnit_SetTransform(Space=LocalSpace, Weight=1, bInitial=False, Value=@"Rotate Aroound Free Pivot_2".Output Transform, Item=@n9.Value)
   n10 = RigVMVariableNode() {variable=Rocker Ball Rotation Channel}
   n11 = &RigVMFunction_MathDoubleRemap(
    SourceMaximum=0
    SourceMinimum=1
    TargetMaximum=0
    bClamp=false
    Value=/Script/ControlRig.RigUnit_GetFloatAnimationChannelFromItem(bInitial=False, Item=RigVMVariableNode() {variable=Rocker Blend Channel}.Value) {is=struct RigVMUnitNode, template="GetAnimationChannelFromItem::Execute(out Value,in Item,in bInitial)"}.Value
    TargetMinimum=/Script/RigVM.RigVMFunction_MathDoubleNegate(Value=&RigUnit_GetFloatAnimationChannelFromItem(bInitial=False, Item=@n10.Value) {template="GetAnimationChannelFromItem::Execute(out Value,in Item,in bInitial)"}.Value) {is=struct RigVMUnitNode, template="Negate::Execute(in Value,out Result)"}.Result
   )
   n12 = RigVMVariableNode() {variable=Foot Rocker}
   > Set Transform_5 = &RigUnit_SetRotation(
    Weight=1
    bInitial=False
    Value=&RigVMFunction_MathQuaternionFromEuler(
     Euler=(X=0,Y=0,Z=0)
     RotationOrder=XYZ
     Euler.X=&RigVMFunction_MathDoubleRemap(SourceMinimum=-360, TargetMaximum=0, TargetMinimum=-360, SourceMaximum=@n11.Result, Value=@n12.Value).Result
    ).Result
    Item=RigVMVariableNode() {variable=Toe Tip Rocker Pivot}.Value
   )
   n13 = RigVMVariableNode() {variable=Foot Rocker}
   > Set Transform_6 = &RigUnit_SetRotation(
    Weight=1
    bInitial=False
    Value=&RigVMFunction_MathQuaternionFromEuler(Euler=(X=0,Y=0,Z=0), RotationOrder=XYZ, Euler.X=&RigVMFunction_MathDoubleClamp(Maximum=360, Value=@n13.Value).Result).Result
    Item=RigVMVariableNode() {variable=Heel Rocker Pivot}.Value
   )
   &RigUnit_SetTransform(
    Weight=1
    bInitial=False
    Value=&RigUnit_GetTransform(Item=(Type=Null,Name="IK"), bInitial=False).Transform
    Item=RigVMVariableNode() {variable=Target Effector}.Value
    ExecutePin=@RigVMFunction_Sequence.D
   )
   n14 = &RigUnit_GetFloatAnimationChannelFromItem(bInitial=False, Item=RigVMVariableNode() {variable=Rocker Ball Rotation Channel}.Value) {template="GetAnimationChannelFromItem::Execute(out Value,in Item,in bInitial)"}
   Multiply = &RigVMFunction_MathDoubleMul(
    A=&RigVMFunction_MathDoubleClamp(
     Maximum=0
     Minimum=&RigVMFunction_MathDoubleNegate(Value=@n14.Value) {template="Negate::Execute(in Value,out Result)"}.Result
     Value=RigVMVariableNode() {variable=Foot Rocker}.Value
    ).Result
    B=&RigUnit_GetFloatAnimationChannelFromItem(bInitial=False, Item=RigVMVariableNode() {variable=Rocker Blend Channel}.Value) {template="GetAnimationChannelFromItem::Execute(out Value,in Item,in bInitial)"}.Value
   )
   n15 = RigVMVariableNode() {variable=Ball Pivot Heel Bend}
   Set Null Local Transform_2 = RigVMFunctionReferenceNode(
    Local=(Rotation=(X=0,Y=0,Z=0,W=1),Translation=(X=0,Y=0,Z=0),Scale3D=(X=1,Y=1,Z=1))
    Local.Rotation=&RigVMFunction_MathQuaternionFromEuler(
     Euler=(X=0,Y=0,Z=0)
     RotationOrder=XYZ
     Euler.X=&RigVMFunction_MathDoubleAdd(B=@Multiply.Result, A=@n15.Value).Result
     Euler.Y=RigVMVariableNode() {variable=Ball Pivot Heel Twist}.Value
     Euler.Z=RigVMVariableNode() {variable=Ball Pivot Heel Side}.Value
    ).Result
    Item=RigVMVariableNode() {variable=Heel Lift Null}.Value
    ExecuteContext=@RigVMFunction_Sequence.C
   )
   &RigVMFunction_MathQuaternionFromEuler(
    Euler=(X=0,Y=0,Z=0)
    RotationOrder=XYZ
    Euler.X=&RigVMFunction_MathDoubleClamp(Maximum=360, Value=RigVMVariableNode() {variable=Foot Rocker}.Value).Result
   )
   GetAnimationChannelFromItem = &RigUnit_GetFloatAnimationChannelFromItem(bInitial=False, Item=RigVMVariableNode() {variable=Rocker Blend Channel}.Value) {template="GetAnimationChannelFromItem::Execute(out Value,in Item,in bInitial)"}
   Negate = &RigVMFunction_MathDoubleNegate(Value=&RigUnit_GetFloatAnimationChannelFromItem(bInitial=False, Item=RigVMVariableNode() {variable=Rocker Ball Rotation Channel}.Value) {template="GetAnimationChannelFromItem::Execute(out Value,in Item,in bInitial)"}.Value) {template="Negate::Execute(in Value,out Result)"}
   Remap = &RigVMFunction_MathDoubleRemap(
    SourceMinimum=-360
    TargetMaximum=0
    TargetMinimum=-360
    SourceMaximum=&RigVMFunction_MathDoubleRemap(SourceMaximum=0, SourceMinimum=1, TargetMaximum=0, bClamp=false, Value=@GetAnimationChannelFromItem.Value, TargetMinimum=@Negate.Result).Result
    Value=RigVMVariableNode() {variable=Foot Rocker}.Value
   )
   &RigVMFunction_MathQuaternionFromEuler(Euler=(X=0,Y=0,Z=0), RotationOrder=XYZ, Euler.X=@Remap.Result)
  ]}
 New Function_ContainedGraph {nodes [
   Entry = RigVMFunctionEntryNode()
   > RigVMFunction_Sequence = RigVMAggregateNode()
   RigVMVariableNode(ExecuteContext=@RigVMFunction_Sequence.A) {variable=Ball Pivot Heel Bend}
   > RigVMVariableNode(Value=0) {variable=Ball Pivot Heel Twist}
   > RigVMVariableNode() {variable=Ball Pivot Heel Side}
   RigVMVariableNode(Value=0, ExecuteContext=@RigVMFunction_Sequence.B) {variable=Foot Rocker}
   > RigVMVariableNode(Value=0) {variable=Foot Bank}
   > RigVMVariableNode() {variable=Ball Pivot}
   RigVMFunctionReturnNode(ExecuteContext=@RigVMFunction_Sequence.C)
  ]}
 New Function_ContainedGraph {nodes [
   Entry = RigVMFunctionEntryNode()
   RigVMFunction_Sequence = &RigVMFunction_Sequence(ExecuteContext=@Entry.ExecuteContext)
   RigVMFunctionReturnNode(ExecuteContext=@RigVMFunction_Sequence.B)
   RigVMFunction_MathQuaternionToEuler = &RigVMFunction_MathQuaternionToEuler(
    RotationOrder=YZX
    Value=&RigUnit_GetTransform(Space=LocalSpace, bInitial=False, Item=RigVMVariableNode() {variable=Ball IK Control}.Value).Transform.Rotation
   )
   RigVMVariableNode(Value=@RigVMFunction_MathQuaternionToEuler.Result.X, ExecuteContext=@RigVMFunction_Sequence.A) {variable=Ball Pivot Heel Bend}
   > RigVMVariableNode(Value=@RigVMFunction_MathQuaternionToEuler.Result.Y) {variable=Ball Pivot Heel Twist}
   > RigVMVariableNode(Value=@RigVMFunction_MathQuaternionToEuler.Result.Z) {variable=Ball Pivot Heel Side}
   n1 = &RigUnit_GetTransform(Space=LocalSpace, bInitial=False, Item=RigVMVariableNode() {variable=Foot Rocker Control}.Value)
   RigVMFunction_MathQuaternionSwingTwist = &RigVMFunction_MathQuaternionSwingTwist(TwistAxis=(X=1,Y=0,Z=0), Input=@n1.Transform.Rotation)
   > RigVMVariableNode(Value=&RigVMFunction_MathQuaternionToEuler(RotationOrder=ZYX, Value=@RigVMFunction_MathQuaternionSwingTwist.Twist).Result.X) {variable=Foot Rocker}
   > RigVMVariableNode(Value=&RigVMFunction_MathQuaternionToEuler(RotationOrder=ZXY, Value=&RigVMFunction_MathQuaternionSwingTwist(TwistAxis=(X=0,Y=1,Z=0), Input=@n1.Transform.Rotation).Twist).Result.Y) {variable=Foot Bank}
   n2 = RigVMVariableNode() {variable=Ball Pivot Control}
   > RigVMVariableNode(Value=&RigVMFunction_MathQuaternionToEuler(RotationOrder=YXZ, Value=&RigUnit_GetTransform(Space=LocalSpace, bInitial=False, Item=@n2.Value).Transform.Rotation).Result.Z) {variable=Ball Pivot}
  ]}
 New Function_ContainedGraph {nodes [
   Entry = RigVMFunctionEntryNode()
   Make = RigVMDispatchNode(
    Elements=(Rotation=(X=0,Y=0,Z=0,W=1),Translation=(X=0,Y=0,Z=0),Scale3D=(X=1,Y=1,Z=1))
    Elements.Scale3D=@Entry.Pivot Transform.Scale3D
    Elements.Translation=@Entry.Pivot Transform.Translation
    Elements.Rotation=@Entry.Rotation Matrix.Rotation
   ) {template="DISPATCH_RigVMDispatch_MakeStruct(in Elements,out Struct)"}
   n1 = &RigVMFunction_MathTransformMul(
    B=(Rotation=(X=0,Y=0,Z=0,W=1),Translation=(X=0,Y=0,Z=0),Scale3D=(X=1,Y=1,Z=1))
    B.Rotation=@Entry.Pivot Transform.Rotation
    A=&RigVMFunction_MathTransformMakeRelative(Global=(Rotation=(X=0,Y=0,Z=0,W=1),Translation=(X=0,Y=0,Z=0),Scale3D=(X=1,Y=1,Z=1)), Parent=@Entry.Pivot Transform).Local
   ) {template="Multiply::Execute(in A,in B,out Result)"}
   > RigVMFunctionReturnNode(Output Transform=&RigVMFunction_MathTransformMul(
     A=&RigVMFunction_MathTransformMakeAbsolute(Parent=@Make.Struct, Local=@n1.Result) {template="Make Absolute::Execute(in Local,in Parent,out Global)"}.Global
     B=@Entry.Input Transform
    ) {template="Multiply::Execute(in A,in B,out Result)"}.Result)
  ]}
 New Function_ContainedGraph {nodes [
   Entry = RigVMFunctionEntryNode()
   > RigUnit_SetRotation = &RigUnit_SetTransform(
    Weight=1
    bInitial=False
    bPropagateToChildren=true
    Value=&RigUnit_GetTransform(bInitial=False, Item=RigVMVariableNode() {variable=Ball Joint}.Value).Transform
    Item=RigVMVariableNode() {variable=Toes FK Control}.Value
   )
   > RigVMFunctionReturnNode()
  ]}
 New Function_ContainedGraph {nodes [
   Entry = RigVMFunctionEntryNode()
   > RigVMFunction_Sequence = RigVMAggregateNode()
   &RigUnit_SetTransform(
    Weight=1
    bInitial=False
    Value=&RigUnit_GetTransform(bInitial=False, Item=RigVMVariableNode() {variable=Ball Joint}.Value).Transform
    ExecutePin=@RigVMFunction_Sequence.B
    Item=RigVMVariableNode() {variable=Toes IK Control}.Value
   )
   RigVMFunctionReturnNode(ExecuteContext=@RigVMFunction_Sequence.C)
   &RigUnit_SetTransform(
    Space=LocalSpace
    Value=(Rotation=(X=0,Y=0,Z=0,W=1),Translation=(X=0,Y=0,Z=0),Scale3D=(X=1,Y=1,Z=1))
    Weight=1
    bInitial=False
    bPropagateToChildren=false
    Item=RigVMVariableNode() {variable=Toe Tip Control}.Value
    ExecutePin=@RigVMFunction_Sequence.A
   )
   > &RigUnit_SetTransform(
    Space=LocalSpace
    Value=(Rotation=(X=0,Y=0,Z=0,W=1),Translation=(X=0,Y=0,Z=0),Scale3D=(X=1,Y=1,Z=1))
    Weight=1
    bInitial=False
    bPropagateToChildren=false
    Item=RigVMVariableNode() {variable=Heel Control}.Value
   )
   > &RigUnit_SetTransform(
    Space=LocalSpace
    Value=(Rotation=(X=0,Y=0,Z=0,W=1),Translation=(X=0,Y=0,Z=0),Scale3D=(X=1,Y=1,Z=1))
    Weight=1
    bInitial=False
    bPropagateToChildren=false
    Item=RigVMVariableNode() {variable=Ball IK Control}.Value
   )
   > &RigUnit_SetTransform(
    Space=LocalSpace
    Value=(Rotation=(X=0,Y=0,Z=0,W=1),Translation=(X=0,Y=0,Z=0),Scale3D=(X=1,Y=1,Z=1))
    Weight=1
    bInitial=False
    bPropagateToChildren=false
    Item=RigVMVariableNode() {variable=Foot Rocker Control}.Value
   )
   > &RigUnit_SetTransform(
    Space=LocalSpace
    Value=(Rotation=(X=0,Y=0,Z=0,W=1),Translation=(X=0,Y=0,Z=0),Scale3D=(X=1,Y=1,Z=1))
    Weight=1
    bInitial=False
    bPropagateToChildren=false
    Item=RigVMVariableNode() {variable=Ball Pivot Control}.Value
   )
   > &RigUnit_SetTransform(
    Space=LocalSpace
    Value=(Rotation=(X=0,Y=0,Z=0,W=1),Translation=(X=0,Y=0,Z=0),Scale3D=(X=1,Y=1,Z=1))
    Weight=1
    bInitial=False
    Item=RigVMVariableNode() {variable=Toes IK Null}.Value
   )
   > Reset Foot Rig Values = RigVMFunctionReferenceNode()
   > Set Foot Pivots_1 = RigVMFunctionReferenceNode()
  ]}
 RigVMFunctionLibrary {nodes [
   Match FK = RigVMCollapseNode(ExecuteContext=())
   Match IK = RigVMCollapseNode(ExecuteContext=())
   Reset Foot Rig Values = RigVMCollapseNode(ExecuteContext=())
   Rotate Aroound Free Pivot = RigVMCollapseNode(
    Input Transform=(Rotation=(X=0,Y=0,Z=0,W=1),Translation=(X=0,Y=0,Z=0),Scale3D=(X=1,Y=1,Z=1))
    Pivot Transform=(Rotation=(X=0,Y=0,Z=0,W=1),Translation=(X=0,Y=0,Z=0),Scale3D=(X=1,Y=1,Z=1))
    Rotation Matrix=(Rotation=(X=0,Y=0,Z=0,W=1),Translation=(X=0,Y=0,Z=0),Scale3D=(X=1,Y=1,Z=1))
   )
   Set Foot Pivots = RigVMCollapseNode(ExecuteContext=())
   Set Foot Values = RigVMCollapseNode(ExecuteContext=())
   Set Null Local Transform = RigVMCollapseNode(ExecuteContext=(), Item=(Type=Null,Name="BK 0"), Local=(Rotation=(X=0,Y=0,Z=0,W=1),Translation=(X=0,Y=0,Z=0),Scale3D=(X=1,Y=1,Z=1)))
  ]}
 RigVMFunction_Sequence_1_ContainedGraph {nodes [
   Entry = RigVMFunctionEntryNode()
   RigVMFunction_Sequence = &RigVMFunction_Sequence(ExecuteContext=@Entry.ExecuteContext)
   n1 = &RigVMFunction_Sequence(ExecuteContext=@RigVMFunction_Sequence.B)
   n2 = &RigVMFunction_Sequence(ExecuteContext=@n1.B)
   n3 = &RigVMFunction_Sequence(ExecuteContext=@n2.B)
   n4 = &RigVMFunction_Sequence(ExecuteContext=@n3.B)
   n5 = &RigVMFunction_Sequence(ExecuteContext=@n4.B)
   n6 = &RigVMFunction_Sequence(ExecuteContext=@n5.B)
   n7 = &RigVMFunction_Sequence(ExecuteContext=@n6.B)
   RigVMFunctionReturnNode(A=@RigVMFunction_Sequence.A, B=@n1.A, C=@n2.A, D=@n3.A, E=@n4.A, F=@n5.A, G=@n6.A, H=@n7.A, I=@n7.B)
  ]}
 RigVMFunction_Sequence_1_ContainedGraph {nodes [
   Entry = RigVMFunctionEntryNode()
   RigVMFunction_Sequence = &RigVMFunction_Sequence(ExecuteContext=@Entry.ExecuteContext)
   n1 = &RigVMFunction_Sequence(ExecuteContext=@RigVMFunction_Sequence.B)
   n2 = &RigVMFunction_Sequence(ExecuteContext=@n1.B)
   n3 = &RigVMFunction_Sequence(ExecuteContext=@n2.B)
   RigVMFunctionReturnNode(A=@RigVMFunction_Sequence.A, B=@n1.A, C=@n2.A, D=@n3.A, E=@n3.B)
  ]}
 RigVMFunction_Sequence_1_ContainedGraph {nodes [
   Entry = RigVMFunctionEntryNode()
   RigVMFunction_Sequence = &RigVMFunction_Sequence(ExecuteContext=@Entry.ExecuteContext)
   n1 = &RigVMFunction_Sequence(ExecuteContext=@RigVMFunction_Sequence.B)
   RigVMFunctionReturnNode(A=@RigVMFunction_Sequence.A, B=@n1.A, C=@n1.B)
  ]}
 RigVMFunction_Sequence_1_ContainedGraph {nodes [
   Entry = RigVMFunctionEntryNode()
   RigVMFunction_Sequence = &RigVMFunction_Sequence(ExecuteContext=@Entry.ExecuteContext)
   n1 = &RigVMFunction_Sequence(ExecuteContext=@RigVMFunction_Sequence.B)
   RigVMFunctionReturnNode(A=@RigVMFunction_Sequence.A, B=@n1.A, C=@n1.B)
  ]}
 RigVMFunction_Sequence_3_ContainedGraph {nodes [
   Entry = RigVMFunctionEntryNode()
   n1 = &RigVMFunction_Sequence(ExecuteContext=@Entry.ExecuteContext)
   n2 = &RigVMFunction_Sequence(ExecuteContext=@n1.B)
   n3 = &RigVMFunction_Sequence(ExecuteContext=@n2.B)
   n4 = &RigVMFunction_Sequence(ExecuteContext=@n3.B)
   RigVMFunctionReturnNode(E=@n3.A, B=@n1.A, F=@n4.A, G=@n4.B, D=@n2.A)
  ]}
 RigVMFunction_Sequence_3_ContainedGraph {nodes [
   Entry = RigVMFunctionEntryNode()
   n1 = &RigVMFunction_Sequence(ExecuteContext=@Entry.ExecuteContext)
   n2 = &RigVMFunction_Sequence(ExecuteContext=@n1.B)
   RigVMFunctionReturnNode(B=@n2.A, C=@n2.B, A=@n1.A)
  ]}
 RigVMModel Backwards Solve Graph {nodes [
   n1 = /Script/ControlRig.RigUnit_InverseExecution() {is=struct RigVMUnitNode}
   RigVMFunction_Sequence = &RigVMFunction_Sequence(ExecuteContext=@n1.ExecutePin)
   Match FK_0 = RigVMFunctionReferenceNode(ExecuteContext=@RigVMFunction_Sequence.A)
   > Match IK_0 = RigVMFunctionReferenceNode()
   /Script/ControlRig.RigUnit_SetTransformItemArray(
    Space=GlobalSpace
    Weight=1
    bInitial=false
    bPropagateToChildren=true
    Transforms=/Script/ControlRig.RigUnit_GetTransformItemArray(Space=GlobalSpace, bInitial=false, Items=RigVMVariableNode() {variable=Toe Joints}.Value) {is=struct RigVMUnitNode}.Transforms
    Items=RigVMVariableNode() {variable=FK Controls}.Value
    ExecutePin=@RigVMFunction_Sequence.B
   ) {is=struct RigVMUnitNode}
  ]}
 RigVMModel Forwards Solve Graph {nodes [
   n1 = /Script/ControlRig.RigUnit_PreBeginExecution() {is=struct RigVMUnitNode}
   DISPATCH_RigDispatch_GetModuleMetadata = RigVMDispatchNode(Default=true, Name=IK Solve, NameSpace=Parent) {template="DISPATCH_RigDispatch_GetModuleMetadata(in Name,in NameSpace,in Default,out Value,out Found)"}
   RigVMVariableNode(Value=@DISPATCH_RigDispatch_GetModuleMetadata.Value, ExecuteContext=@n1.ExecutePin) {variable=IK Solve}
   > RigVMFunction_Sequence_2 = RigVMAggregateNode()
   For_Each_1 = RigVMDispatchNode(ExecuteContext=@RigVMFunction_Sequence_2.A, Array=RigVMVariableNode() {variable=IK Controls}.Value) {template="DISPATCH_RigVMDispatch_ArrayIterator(in Array,out Element,out Index,out Count,out Ratio)"}
   n2 = RigVMVariableNode() {variable=Pivot Controls}
   GetAnimationChannelFromItem = /Script/ControlRig.RigUnit_GetBoolAnimationChannelFromItem(bInitial=False, Item=RigVMVariableNode() {variable=Pivot Control Vis Channel}.Value) {is=struct RigVMUnitNode, template="GetAnimationChannelFromItem::Execute(out Value,in Item,in bInitial)"}
   n3 = RigVMVariableNode() {variable=IK Solve}
   > SetControlVisibility = /Script/ControlRig.RigUnit_SetControlVisibility(
    Item=@For_Each_1.Element
    bVisible=RigVMDispatchNode(
     Condition=RigVMDispatchNode(Element=@For_Each_1.Element, Array=@n2.Value) {template="DISPATCH_RigVMDispatch_ArrayFind(in Array,in Element,out Index,out Success)"}.Success
     True=&RigVMFunction_MathBoolAnd(B=@GetAnimationChannelFromItem.Value, A=@n3.Value).Result
     False=@n3.Value
    ) {template="DISPATCH_RigVMDispatch_If(in Condition,in True,in False,out Result)"}.Result
   ) {is=struct RigVMUnitNode, template="SetControlVisibility::Execute(in Item,in Pattern,in bVisible)"}
   n4 = RigVMVariableNode() {variable=IK Solve}
   &RigUnit_SetControlVisibility(ExecutePin=@RigVMFunction_Sequence_2.B, Item=RigVMVariableNode() {variable=Toes IK Control}.Value, bVisible=@n4.Value) {template="SetControlVisibility::Execute(in Item,in Pattern,in bVisible)"}
   > &RigUnit_SetControlVisibility(bVisible=&RigVMFunction_MathBoolNot(Value=@n4.Value).Result, Item=RigVMVariableNode() {variable=Toes FK Control}.Value) {template="SetControlVisibility::Execute(in Item,in Pattern,in bVisible)"}
   Set Foot Values = RigVMFunctionReferenceNode(ExecuteContext=@RigVMFunction_Sequence_2.C)
   RigVMFunction_ControlFlowBranch = &RigVMFunction_ControlFlowBranch(ExecuteContext=@"Set Foot Values".ExecuteContext, Condition=RigVMVariableNode() {variable=IK Solve}.Value)
   Set Foot Pivots = RigVMFunctionReferenceNode(ExecuteContext=@RigVMFunction_ControlFlowBranch.True)
   RigUnit_BeginExecution = &RigUnit_BeginExecution()
   RigVMFunction_Sequence_1 = RigVMAggregateNode(ExecuteContext=@RigUnit_BeginExecution.ExecutePin)
   n5 = &RigVMFunction_ControlFlowBranch(ExecuteContext=@RigVMFunction_Sequence_1.D, Condition=RigVMVariableNode() {variable=IK Solve}.Value)
   n6 = &RigUnit_SetTransform(
    Weight=1
    bInitial=False
    Value=&RigUnit_GetTransform(bInitial=False, Item=RigVMVariableNode() {variable=Toes FK Control}.Value).Transform
    Item=RigVMVariableNode() {variable=Ball Joint}.Value
    ExecutePin=@n5.False
   )
   Get_Module_Metadata_5 = RigVMDispatchNode(Default=False, Name=Match IK, NameSpace=Parent) {template="DISPATCH_RigDispatch_GetModuleMetadata(in Name,in NameSpace,in Default,out Value,out Found)"}
   n7 = &RigVMFunction_ControlFlowBranch(Condition=@Get_Module_Metadata_5.Value, ExecuteContext=@n6.ExecutePin)
   Match IK_2 = RigVMFunctionReferenceNode(ExecuteContext=@n7.True)
   Make_Array = RigVMDispatchNode(
    Values=((Type=None,Name="None"),(Type=None,Name="None"))
    Values.1=RigVMVariableNode() {variable=Ball Joint}.Value
    Values.0=RigVMVariableNode() {variable=Foot Joint}.Value
   ) {template="DISPATCH_RigVMDispatch_ArrayMake(in Values,out Array)"}
   /Script/ControlRig.RigUnit_FABRIKItemArray(
    MaxIterations=10
    Precision=0.0001
    Weight=1
    bPropagateToChildren=True
    bSetEffectorTransform=false
    EffectorTransform=&RigUnit_GetTransform(bInitial=False, Item=RigVMVariableNode() {variable=Ball IK Null}.Value).Transform
    Items=@Make_Array.Array
    ExecutePin=@n5.True
   ) {is=struct RigVMUnitNode}
   n8 = RigVMVariableNode() {variable=Toe IK Null}
   > &RigUnit_AimConstraintLocalSpaceOffset(
    Filter=(bX=True,bY=True,bZ=True)
    Parents=((Item=(Type=Bone,Name="None"),Weight=1))
    Weight=1
    WorldUp=(Target=(X=0,Y=0,Z=1),Kind=Location,Space=(Type=None,Name="None"))
    Child=RigVMVariableNode() {variable=Ball Joint}.Value
    Parents.0.Item=@n8.Value
    WorldUp.Space=@n8.Value
   )
   n9 = RigVMVariableNode() {variable=Ball Joint}
   n10 = RigVMVariableNode() {variable=Toes IK Null}
   > &RigUnit_SetTransform(
    Weight=1
    bInitial=False
    Value=/Script/ControlRig.RigUnit_ProjectTransformToNewParent(bChildInitial=True, bNewParentInitial=False, bOldParentInitial=True, NewParent=@n9.Value, OldParent=@n9.Value, Child=@n10.Value) {is=struct RigVMUnitNode}.Transform
    Item=@n10.Value
   )
   > n11 = &RigUnit_SetTransform(
    Weight=1
    bInitial=False
    Value=&RigUnit_GetTransform(bInitial=False, Item=RigVMVariableNode() {variable=Toes IK Control}.Value).Transform
    Item=RigVMVariableNode() {variable=Ball Joint}.Value
   )
   Get_Module_Metadata_2 = RigVMDispatchNode(Default=False, Name=Match FK, NameSpace=Parent) {template="DISPATCH_RigDispatch_GetModuleMetadata(in Name,in NameSpace,in Default,out Value,out Found)"}
   n12 = &RigVMFunction_ControlFlowBranch(Condition=@Get_Module_Metadata_2.Value, ExecuteContext=@n11.ExecutePin)
   Match FK_1 = RigVMFunctionReferenceNode(ExecuteContext=@n12.True)
   Get_Module_Metadata_3 = RigVMDispatchNode(Default=False, Name=Key Controls, NameSpace=Parent) {template="DISPATCH_RigDispatch_GetModuleMetadata(in Name,in NameSpace,in Default,out Value,out Found)"}
   n13 = &RigVMFunction_ControlFlowBranch(Condition=@Get_Module_Metadata_3.Value, ExecuteContext=@RigVMFunction_Sequence_1.G)
   RigVMInvokeEntryNode(Entry=Key Controls, ExecuteContext=@n13.True)
   n14 = &RigVMFunction_ControlFlowBranch(Condition=RigVMVariableNode() {variable=Has Toes}.Value, ExecuteContext=@RigVMFunction_Sequence_1.F)
   For_Each = RigVMDispatchNode(ExecuteContext=@n14.True, Array=RigVMVariableNode() {variable=Toe Joints}.Value) {template="DISPATCH_RigVMDispatch_ArrayIterator(in Array,out Element,out Index,out Count,out Ratio)"}
   n15 = RigVMDispatchNode(Index=@For_Each.Index, Array=RigVMVariableNode() {variable=FK Controls}.Value) {template="DISPATCH_RigVMDispatch_ArrayGetAtIndex(in Array,in Index,out Element)"}
   > &RigUnit_SetTransform(
    Weight=1
    bInitial=False
    Item=@For_Each.Element
    Value=&RigUnit_ProjectTransformToNewParent(bChildInitial=True, bNewParentInitial=False, bOldParentInitial=True, NewParent=@n15.Element, OldParent=@n15.Element, Child=@For_Each.Element).Transform
   )
   RigVMFunction_MathBoolAnd = &RigVMFunction_MathBoolAnd(
    A=&RigUnit_GetBoolAnimationChannelFromItem(bInitial=False, Item=RigVMVariableNode() {variable=Footprint Vis Animation Channel}.Value) {template="GetAnimationChannelFromItem::Execute(out Value,in Item,in bInitial)"}.Value
    B=RigVMVariableNode() {variable=IK Solve}.Value
   )
   &RigUnit_SetControlVisibility(
    bVisible=@RigVMFunction_MathBoolAnd.Result
    Item=RigVMVariableNode() {variable=Footprint Control}.Value
    ExecutePin=@RigVMFunction_Sequence_1.B
   ) {template="SetControlVisibility::Execute(in Item,in Pattern,in bVisible)"}
   n16 = /Script/ControlRig.RigUnit_IsInteracting() {is=struct RigVMUnitNode}
   n17 = &RigVMFunction_ControlFlowBranch(Condition=@n16.bIsInteracting, ExecuteContext=@RigVMFunction_Sequence_1.E)
   n18 = RigVMVariableNode() {variable=Foot Joint}
   n19 = RigVMVariableNode() {variable=Foot Rocker Control}
   Get_Metadata_2 = RigVMDispatchNode(Default=(Rotation=(X=0,Y=0,Z=0,W=1),Translation=(X=0,Y=0,Z=0),Scale3D=(X=1,Y=1,Z=1)), Name=Delta, NameSpace=Self, Item=@n19.Value) {template="DISPATCH_RigDispatch_GetMetadata(in Item,in Name,in NameSpace,in Default,out Value,out Found)"}
   n20 = /Script/RigVM.RigVMFunction_MathTransformInverse(Value=&RigUnit_GetTransform(Space=LocalSpace, bInitial=False, Item=@n19.Value).Transform) {is=struct RigVMUnitNode, template="Inverse::Execute(in Value,out Result)"}
   n21 = /Script/ControlRig.RigUnit_SetControlOffset(
    Space=GlobalSpace
    Offset=&RigVMFunction_MathTransformMakeAbsolute(
     Parent=&RigUnit_GetTransform(bInitial=False, Item=@n18.Value).Transform
     Local=&RigVMFunction_MathTransformMul(B=@Get_Metadata_2.Value, A=@n20.Result) {template="Multiply::Execute(in A,in B,out Result)"}.Result
    ) {template="Make Absolute::Execute(in Local,in Parent,out Global)"}.Global
    Control=&RigUnit_ItemToName(Value=@n19.Value) {template="Cast::Execute(in Value,out Result)"}.Result
    ExecutePin=@n17.False
   ) {is=struct RigVMUnitNode, template="SetControlOffset::Execute(in Control,io Offset,in Space)"}
   n22 = &RigVMFunction_ControlFlowBranch(
    Condition=RigVMDispatchNode(Array=@n16.Items, Element=RigVMVariableNode() {variable=Heel Control}.Value) {template="DISPATCH_RigVMDispatch_ArrayFind(in Array,in Element,out Index,out Success)"}.Success
    ExecuteContext=@n21.ExecutePin
   )
   &RigUnit_SetTransform(
    Weight=1
    bInitial=False
    bPropagateToChildren=false
    Value=&RigUnit_GetTransform(bInitial=False, Item=RigVMVariableNode() {variable=Toe IK Null}.Value).Transform
    ExecutePin=@n22.True
    Item=RigVMVariableNode() {variable=Toe Tip Pivot Null}.Value
   )
  ]}
 RigVMModel Utils {nodes [
   n1 = /Script/RigVM.RigVMFunction_UserDefinedEvent(EventName=Key Controls) {is=struct RigVMUnitNode}
   n2 = RigVMDispatchNode(
    Array=RigVMDispatchNode(Array=RigVMVariableNode() {variable=FK Controls}.Value) {template="DISPATCH_RigVMDispatch_ArrayClone(in Array,out Clone)"}.Clone
    ExecuteContext=@n1.ExecutePin
    Other=RigVMVariableNode() {variable=IK Controls}.Value
   ) {template="DISPATCH_RigVMDispatch_ArrayAppend(io Array,in Other)"}
   > For_Each_2 = RigVMDispatchNode(Array=@n2.Array) {template="DISPATCH_RigVMDispatch_ArrayIterator(in Array,out Element,out Index,out Count,out Ratio)"}
   > SendEvent = /Script/ControlRig.RigUnit_SendEvent(Event=RequestAutoKey, OffsetInSeconds=0, bEnable=True, bOnlyDuringInteraction=false, Item=@For_Each_2.Element) {is=struct RigVMUnitNode, template="SendEvent::Execute(in Event,in Item,in OffsetInSeconds,in bEnable,in bOnlyDuringInteraction)"}
   &RigUnit_SendEvent(
    Event=RequestAutoKey
    OffsetInSeconds=0
    bEnable=True
    bOnlyDuringInteraction=false
    ExecutePin=@For_Each_2.Completed
    Item=RigVMVariableNode() {variable=Toes FK Control}.Value
   ) {template="SendEvent::Execute(in Event,in Item,in OffsetInSeconds,in bEnable,in bOnlyDuringInteraction)"}
   > &RigUnit_SendEvent(
    Event=RequestAutoKey
    OffsetInSeconds=0
    bEnable=True
    bOnlyDuringInteraction=false
    Item=RigVMVariableNode() {variable=Toes IK Control}.Value
   ) {template="SendEvent::Execute(in Event,in Item,in OffsetInSeconds,in bEnable,in bOnlyDuringInteraction)"}
  ]}
]
hierarchy [
 Foot Joint {type=Connector, loc=<0 0 0>, rot=<0 0 0 1>}
 Root {type=Socket, loc=<0 0 0>, rot=<0 0 0 1>}
 Ball Joint {type=Connector, loc=<0 0 0>, rot=<0 0 0 1>}
 Toe Joints {type=Connector, loc=<0 0 0>, rot=<0 0 0 1>}
 Toe Tip Pivot {type=Connector, loc=<0 0 0>, rot=<0 0 0 1>}
 Heel Pivot {type=Connector, loc=<0 0 0>, rot=<0 0 0 1>}
 Inner Bank Pivot {type=Connector, loc=<0 0 0>, rot=<0 0 0 1>}
 Outer Bank Pivot {type=Connector, loc=<0 0 0>, rot=<0 0 0 1>}
]
variables [
 Ball IK Control {type=FRigElementKey, default=(Type=Bone,Name="")}
 Ball IK Null {type=FRigElementKey, default=(Type=Bone,Name="")}
 Ball IK Shape Settings {
  type=FRigUnit_HierarchyAddControl_ShapeSettings
  default=(bVisible=True,Name="Box_Thick",Color=(R=1,G=1,B=1,A=1),Transform=(Rotation=(X=-0.7071,Y=0,Z=0,W=0.7071),Translation=(X=0,Y=0,Z=1),Scale3D=(X=1.5,Y=0.3,Z=0.4)))
 }
 Ball Joint {type=FRigElementKey, default=(Type=None,Name="")}
 Ball Pivot {type=float, default=0}
 Ball Pivot Bone {type=FRigElementKey, default=(Type=None,Name="")}
 Ball Pivot Control {type=FRigElementKey, default=(Type=Bone,Name="")}
 Ball Pivot Heel Bend {type=float, default=0}
 Ball Pivot Heel Side {type=float, default=0}
 Ball Pivot Heel Twist {type=float, default=0}
 Ball Pivot Null {type=FRigElementKey, default=(Type=None,Name="")}
 Channels Control {type=FRigElementKey, default=(Type=Bone,Name="")}
 Control Scale {type=double, default=1}
 FK Control Mirror Behavior {type=bool, default=True}
 FK Controls {type="TArray<FRigElementKey>"}
 Foot Bank {type=float, default=0}
 Foot Joint {type=FRigElementKey, default=(Type=None,Name="")}
 Foot Rocker {type=float, default=0}
 Foot Rocker Control {type=FRigElementKey, default=(Type=Bone,Name="")}
 Foot Rocker Shape Settings {
  type=FRigUnit_HierarchyAddControl_ShapeSettings
  default=(bVisible=True,Name="Sphere_Thick",Color=(R=1,G=1,B=1,A=1),Transform=(Rotation=(X=0,Y=0,Z=0,W=1),Translation=(X=0,Y=0,Z=5),Scale3D=(X=0.5,Y=0.5,Z=0.5)))
 }
 Footprint Control {type=FRigElementKey, default=(Type=Bone,Name="")}
 Footprint Space Transform {type=FTransform, default=(Rotation=(X=0,Y=0,Z=0,W=1),Translation=(X=0,Y=0,Z=0),Scale3D=(X=1,Y=1,Z=1))}
 Footprint Vis Animation Channel {type=FRigElementKey, default=(Type=Bone,Name="")}
 Has Toes {type=bool, default=False}
 Heel Control {type=FRigElementKey, default=(Type=Bone,Name="")}
 Heel Lift Null {type=FRigElementKey, default=(Type=None,Name="")}
 Heel Pivot Null {type=FRigElementKey, default=(Type=Bone,Name="")}
 Heel Pivot Offset {type=FVector, default=(X=0,Y=-22,Z=0)}
 Heel Rocker Pivot {type=FRigElementKey, default=(Type=Bone,Name="")}
 IK Controls {type="TArray<FRigElementKey>"}
 IK Foot Space {type=FRigElementKey, default=(Type=Bone,Name="")}
 IK Null {type=FRigElementKey, default=(Type=Bone,Name="")}
 IK Solve {type=bool, default=False}
 Inner Bank Offset {type=FVector, default=(X=5,Y=0,Z=0)}
 Inner Pivot Null {type=FRigElementKey, default=(Type=Bone,Name="")}
 Negative Side {type=bool, default=False}
 Outer Bank Offset {type=FVector, default=(X=-6,Y=0,Z=0)}
 Outer Pivot Null {type=FRigElementKey, default=(Type=Bone,Name="")}
 Pivot Control Vis Channel {type=FRigElementKey, default=(Type=Bone,Name="")}
 Pivot Controls {type="TArray<FRigElementKey>"}
 Pivot Shape Settings {
  type=FRigUnit_HierarchyAddControl_ShapeSettings
  default=(bVisible=True,Name="Sphere_Solid",Color=(R=1,G=1,B=1,A=1),Transform=(Rotation=(X=-0.7071,Y=0,Z=0,W=0.7071),Translation=(X=0,Y=0,Z=1),Scale3D=(X=0.2,Y=0.2,Z=0.2)))
 }
 Rocker Ball Rotation Channel {type=FRigElementKey, default=(Type=Bone,Name="")}
 Rocker Blend Channel {type=FRigElementKey, default=(Type=Bone,Name="")}
 Roll Blend {type=double, default=0}
 Running Parent {type=FRigElementKey, default=(Type=Bone,Name="")}
 Target Effector {type=FRigElementKey, default=(Type=Bone,Name="")}
 Target IK Control {type=FRigElementKey, default=(Type=Bone,Name="")}
 Target IK Driver {type=FRigElementKey, default=(Type=Bone,Name="")}
 Target IK Orientation {type=FQuat, default=(X=0,Y=0,Z=0,W=1)}
 Toe IK End Null {type=FRigElementKey, default=(Type=Bone,Name="")}
 Toe IK Null {type=FRigElementKey, default=(Type=Bone,Name="")}
 Toe Joints {type="TArray<FRigElementKey>"}
 Toe Shape Settings {
  type=FRigUnit_HierarchyAddControl_ShapeSettings
  default=(bVisible=True,Name="HalfCircle_Thick",Color=(R=1,G=1,B=1,A=1),Transform=(Rotation=(X=-0.7071,Y=0,Z=-0.7071,W=0),Translation=(X=-3,Y=0,Z=0),Scale3D=(X=1,Y=1,Z=1)))
 }
 Toe Tip Control {type=FRigElementKey, default=(Type=Bone,Name="")}
 Toe Tip Pivot Null {type=FRigElementKey, default=(Type=None,Name="")}
 Toe Tip Pivot Offset {type=FVector, default=(X=0,Y=8,Z=0)}
 Toe Tip Rocker Pivot {type=FRigElementKey, default=(Type=Bone,Name="")}
 Toes FK Control {type=FRigElementKey, default=(Type=Bone,Name="")}
 Toes FK Null {type=FRigElementKey, default=(Type=Bone,Name="")}
 Toes IK Control {type=FRigElementKey, default=(Type=Bone,Name="")}
 Toes IK Null {type=FRigElementKey, default=(Type=Bone,Name="")}
]
